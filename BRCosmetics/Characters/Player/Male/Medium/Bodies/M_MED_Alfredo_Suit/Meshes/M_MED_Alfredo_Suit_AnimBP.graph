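# Animation Blueprint for the Alfredo Suit male medium body cosmetic, layering corrective rigs and pouch physics onto the player's pose.
# The parent mesh pose passes through a deform Control Rig, six chained AnimDynamics nodes for left and right pouches, then a second Control Rig.
# Each update derives that rig's weapon, hand-glider and celebration flags from the player anim instance, weapon type, glider type and montage state, alongside inherited parachute and front-end flags.
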AnimBlueprint /BRCosmetics/Characters/Player/Male/Medium/Bodies/M_MED_Alfredo_Suit/Meshes/M_MED_Alfredo_Suit_AnimBP
target_skeleton=&/M_MED_Alfredo_Suit_Skeleton
graphs [
 AnimGraph {nodes [
   n1 = Control Rig(Source=Copy Pose From Mesh() {bUseAttachedParent=true}) {
    alphaBoolBlend {blendOption=Linear}
    controlRigAssetReference {blueprintRigClass=&/M_MED_Alfredo_Suit_DeformRig_ControlRig.M_MED_Alfredo_Suit_DeformRig_ControlRig_C}
   }
   n2 = Anim Dynamics - Bone: dyn_pouch_front_l(ComponentPose=Local To Component(LocalPose=@n1), Alpha=1, ExternalForce="0,0,0") {
    angularDampingOverride=0.98
    angularSpringConstant=200
    bAngularSpring=true
    bOverrideAngularDamping=true
    bUseGravityOverride=true
    simulationSpace=RootRelative
    alphaBoolBlend {blendOption=Linear}
    boundBone {boneName=dyn_pouch_front_l}
    componentAppliedLinearAccClamp=<2500 2500 2500>
    componentLinearAccScale=<0.11 0.11 0.11>
    componentLinearVelScale=<0.95 0.95 0.95>
    gravityOverride {z=-250}
    physicsBodyDefinitions [{
      boundBone {boneName=dyn_pouch_front_l}
      boxExtents=<15 15 15>
      constraintSetup {angularLimitsMax=<2.5 2.5 5>, angularLimitsMin=<-2.5 -2.5 -5>, angularTarget=<1 0 0>}
      localJointOffset=<0 1 0>
     }]
   }
   n3 = Anim Dynamics - Bone: dyn_pouch_mid_l(ComponentPose=@n2, Alpha=1, ExternalForce="0,0,0") {
    angularDampingOverride=0.985
    angularSpringConstant=200
    bAngularSpring=true
    bOverrideAngularDamping=true
    bUseGravityOverride=true
    simulationSpace=RootRelative
    alphaBoolBlend {blendOption=Linear}
    boundBone {boneName=dyn_pouch_mid_l}
    componentAppliedLinearAccClamp=<2500 2500 2500>
    componentLinearAccScale=<0.9 0.9 0.9>
    componentLinearVelScale=<0.9 0.9 0.9>
    gravityOverride {z=-250}
    physicsBodyDefinitions [{
      boundBone {boneName=dyn_pouch_mid_l}
      boxExtents=<16 16 16>
      constraintSetup {angularLimitsMax=<2.5 2.5 5>, angularLimitsMin=<-2.5 -2.5 -5>, angularTarget=<1 0 0>}
      localJointOffset=<0 1 0>
     }]
   }
   n4 = Anim Dynamics - Bone: dyn_pouch_back_l(ComponentPose=@n3, Alpha=1, ExternalForce="0,0,0") {
    angularDampingOverride=0.99
    angularSpringConstant=200
    bAngularSpring=true
    bOverrideAngularDamping=true
    bUseGravityOverride=true
    simulationSpace=RootRelative
    alphaBoolBlend {blendOption=Linear}
    boundBone {boneName=dyn_pouch_back_l}
    componentAppliedLinearAccClamp=<2500 2500 2500>
    componentLinearAccScale=<0.1 0.1 0.1>
    componentLinearVelScale=<1 1 1>
    gravityOverride {z=-250}
    physicsBodyDefinitions [{
      boundBone {boneName=dyn_pouch_back_l}
      boxExtents=<17 17 17>
      constraintSetup {angularLimitsMax=<2.5 2.5 5>, angularLimitsMin=<-2.5 -2.5 -5>, angularTarget=<1 0 0>}
      localJointOffset=<0 1 0>
     }]
   }
   n5 = Anim Dynamics - Bone: dyn_pouch_front_r(ComponentPose=@n4, Alpha=1, ExternalForce="0,0,0") {
    angularDampingOverride=0.98
    angularSpringConstant=200
    bAngularSpring=true
    bOverrideAngularDamping=true
    bUseGravityOverride=true
    simulationSpace=RootRelative
    alphaBoolBlend {blendOption=Linear}
    boundBone {boneName=dyn_pouch_front_r}
    componentAppliedLinearAccClamp=<2500 2500 2500>
    componentLinearAccScale=<0.11 0.11 0.11>
    componentLinearVelScale=<0.95 0.95 0.95>
    gravityOverride {z=-250}
    physicsBodyDefinitions [{
      boundBone {boneName=dyn_pouch_front_r}
      boxExtents=<15 15 15>
      constraintSetup {angularLimitsMax=<2.5 2.5 5>, angularLimitsMin=<-2.5 -2.5 -5>, angularTarget=<1 0 0>}
      localJointOffset=<0 -1 0>
     }]
   }
   n6 = Anim Dynamics - Bone: dyn_pouch_mid_r(ComponentPose=@n5, Alpha=1, ExternalForce="0,0,0") {
    angularDampingOverride=0.985
    angularSpringConstant=200
    bAngularSpring=true
    bOverrideAngularDamping=true
    bUseGravityOverride=true
    simulationSpace=RootRelative
    alphaBoolBlend {blendOption=Linear}
    boundBone {boneName=dyn_pouch_mid_r}
    componentAppliedLinearAccClamp=<2500 2500 2500>
    componentLinearAccScale=<0.9 0.9 0.9>
    componentLinearVelScale=<0.9 0.9 0.9>
    gravityOverride {z=-250}
    physicsBodyDefinitions [{
      boundBone {boneName=dyn_pouch_mid_r}
      boxExtents=<16 16 16>
      constraintSetup {angularLimitsMax=<2.5 2.5 5>, angularLimitsMin=<-2.5 -2.5 -5>, angularTarget=<1 0 0>}
      localJointOffset=<0 -1 0>
     }]
   }
   n7 = Anim Dynamics - Bone: dyn_pouch_back_r(ComponentPose=@n6, Alpha=1, ExternalForce="0,0,0") {
    angularDampingOverride=0.99
    angularSpringConstant=200
    bAngularSpring=true
    bOverrideAngularDamping=true
    bUseGravityOverride=true
    simulationSpace=RootRelative
    alphaBoolBlend {blendOption=Linear}
    boundBone {boneName=dyn_pouch_back_r}
    componentAppliedLinearAccClamp=<2500 2500 2500>
    componentLinearAccScale=<0.1 0.1 0.1>
    componentLinearVelScale=<1 1 1>
    gravityOverride {z=-250}
    physicsBodyDefinitions [{
      boundBone {boneName=dyn_pouch_back_r}
      boxExtents=<17 17 17>
      constraintSetup {angularLimitsMax=<2.5 2.5 5>, angularLimitsMin=<-2.5 -2.5 -5>, angularTarget=<1 0 0>}
      localJointOffset=<0 -1 0>
     }]
   }
   n8 = Control Rig!(
    Source=Component To Local(ComponentPose=@n7)
    Alpha
    alt_weapon:bool=Get alt_weapon() -> (alt_weapon:bool)
    is_parachute_opened:bool=Get bIsParachuteOpened()
    has_hand_glider:bool=Get has_hand_glider() -> (has_hand_glider:bool)
    is_playing_celebration:bool=Get is_playing_celebration() -> (is_playing_celebration:bool)
    is_front_end:bool=Get bIsInFrontEnd()
   ) -> (Pose) {
    alphaBoolBlend {blendOption=Linear}
    controlRigAssetReference {blueprintRigClass=&/M_MED_Alfredo_Suit_CtrlRig.M_MED_Alfredo_Suit_CtrlRig_C}
   }
   Output Pose(Result=@n8) {layerGroup=None}
  ]}
 EventGraph {nodes [
   Event BlueprintUpdateAnimation()
   > n1 = Sequence!(execute) -> (then_0, then_1, then_2:Exec)
   n2 = GetAnimInstance(self=GetOwnerSkeletalMeshComponent())
   > n3 = Cast To FortPlayerAnimInstance(Object=@n2)
   > Set alt_weapon(alt_weapon:bool=false) -> (Output_Get:bool)
   > n4 = Switch on EFortWeaponCoreAnimation!(execute, Selection:EFortWeaponCoreAnimation Enum=Get WeaponCoreAnim(self=@n3)) -> (
    Melee:Exec
    Pistol:Exec
    Shotgun:Exec
    PaperBlueprint:Exec
    Rifle:Exec
    MeleeOneHand:Exec
    MachinePistol:Exec
    RocketLauncher:Exec
    GrenadeLauncher:Exec
    GoingCommando:Exec
    AssaultRifle:Exec
    TacticalShotgun:Exec
    SniperRifle:Exec
    TrapPlacement:Exec
    ShoulderLauncher:Exec
    AbilityDecoTool:Exec
    Crossbow:Exec
    C4:Exec
    RemoteControl:Exec
    DualWield:Exec
    AR_BullPup:Exec
    AR_ForwardGrip:Exec
    MedPackPaddles:Exec
    SMG_P90:Exec
    AR_DrumGun:Exec
    Consumable_Small:Exec
    Consumable_Large:Exec
    Balloon:Exec
    MountedTurret:Exec
    CreativeTool:Exec
    ExplosiveBow:Exec
    AshtonIndigo:Exec
    AshtonChicago:Exec
    MeleeDualWield:Exec
    Unarmed:Exec
   )
   Branch(execute=[@n4.AssaultRifle, @n4.Rifle, @n4.RocketLauncher, @n4.Shotgun, @n4.SniperRifle, @n4.TacticalShotgun])
   > Set alt_weapon(alt_weapon=true)
   Is Valid(exec=@n1.then_1, InputObject=@n2)
   > Set is_playing_celebration(is_playing_celebration:bool=Montage_IsActive!(self:Anim Instance Object Reference=@n2, Montage:Anim Montage Object Reference=/Game/Animation/Game/MainPlayer/Emotes/Emote_Celebration_M) -> (ReturnValue:bool)) -> (Output_Get:bool)
   n5 = Get fort_player() -> (fort_player:Fort Player Anim Instance Object Reference)
   Switch on EGliderType(
    execute=@n1.then_2
    Selection=Break Fort Anim Input Player Glider Anim Asset!(FortAnimInput_PlayerGliderAnimAsset:"Fort Anim Input Player Glider Anim Asset Structure (by ref)"=Get CurrentGliderAnimSet!(self:Fort Player Anim Instance Object Reference=@n5) -> (CurrentGliderAnimSet:Fort Anim Input Player Glider Anim Asset Structure)) -> (
     Default_Base_BS:Blend Space Object Reference
     Default_BodyAdditive_MaleMedium_BS:Blend Space Object Reference
     Default_BodyAdditive_MaleLarge_BS:Blend Space Object Reference
     Default_BodyAdditive_FemaleSmall_BS:Blend Space Object Reference
     Default_BodyAdditive_FemaleMedium_BS:Blend Space Object Reference
     Default_BodyAdditive_FemaleLarge_BS:Blend Space Object Reference
     Default_TurnAdditive_MaleMedium_BS:Blend Space Object Reference
     Default_TurnAdditive_MaleLarge_BS:Blend Space Object Reference
     Default_TurnAdditive_FemaleSmall_BS:Blend Space Object Reference
     Default_TurnAdditive_FemaleMedium_BS:Blend Space Object Reference
     Default_TurnAdditive_FemaleLarge_BS:Blend Space Object Reference
     Into_Base_BS:Blend Space Object Reference
     Into_Base_MaleMedium_BS:Blend Space Object Reference
     Into_BodyAdditive_MaleMedium_BS:Blend Space Object Reference
     Into_BodyAdditive_MaleLarge_BS:Blend Space Object Reference
     Into_BodyAdditive_FemaleSmall_BS:Blend Space Object Reference
     Into_BodyAdditive_FemaleMedium_BS:Blend Space Object Reference
     Into_BodyAdditive_FemaleLarge_BS:Blend Space Object Reference
     Into_TurnAdditive_MaleMedium_BS:Blend Space Object Reference
     Into_TurnAdditive_MaleLarge_BS:Blend Space Object Reference
     Into_TurnAdditive_FemaleSmall_BS:Blend Space Object Reference
     Into_TurnAdditive_FemaleMedium_BS:Blend Space Object Reference
     Into_TurnAdditive_FemaleLarge_BS:Blend Space Object Reference
     Lean_MaleMedium_BS:Blend Space Object Reference
     Lean_MaleLarge_BS:Blend Space Object Reference
     Lean_FemaleSmall_BS:Blend Space Object Reference
     Lean_FemaleMedium_BS:Blend Space Object Reference
     Lean_FemaleLarge_BS:Blend Space Object Reference
     LeanAdditive_Center_MaleMedium_Pose:Anim Sequence Object Reference
     LeanAdditive_Center_MaleLarge_Pose:Anim Sequence Object Reference
     LeanAdditive_Center_FemaleSmall_Pose:Anim Sequence Object Reference
     LeanAdditive_Center_FemaleMedium_Pose:Anim Sequence Object Reference
     LeanAdditive_Center_FemaleLarge_Pose:Anim Sequence Object Reference
     LeanAdditive_Into_BS:Blend Space Object Reference
     LeanAdditive_ForwardInto_Anim:Anim Sequence Object Reference
     LeanAdditive_ForwardInto_FromDeploy_Anim:Anim Sequence Object Reference
     LeanAdditive_BackInto_Anim:Anim Sequence Object Reference
     LeanAdditive_LeftInto_Anim:Anim Sequence Object Reference
     LeanAdditive_RightInto_Anim:Anim Sequence Object Reference
     LeanAdditive_ForwardOut_Anim:Anim Sequence Object Reference
     LeanAdditive_BackOut_Anim:Anim Sequence Object Reference
     LeanAdditive_LeftOut_Anim:Anim Sequence Object Reference
     LeanAdditive_RightOut_Anim:Anim Sequence Object Reference
     ToGlide_BS:Blend Space Object Reference
     ToGlide_Lean_BS:Blend Space Object Reference
     ToDive_BS:Blend Space Object Reference
     ToDive_Lean_BS:Blend Space Object Reference
     Dive_WeaponR_Additive_BS:Blend Space Object Reference
     Glide_WeaponR_Additive_Anim:Anim Sequence Object Reference
     GenericAdditive_Male_BS:Blend Space Object Reference
     GenericAdditive_Female_BS:Blend Space Object Reference
     RootModPitchMin:float=~
     RootModPitchMax:float=~
     RootModYOffsetMin:float=~
     RootModYOffsetMax:float=~
     PelvisModVertical:Pelvis Mod Vertical Input Spring Structure
     PelvisModLateral:Pelvis Mod Lateral Input Spring Structure
     PelvisModBankAngle:Pelvis Mod Bank Angle Input Spring Structure
     ThighModLegAngle:Thigh Mod Leg Angle Input Structure
     ThighModLegPitch:Thigh Mod Leg Pitch Spring Structure
     ThighModLegBank:Thigh Mod Leg Bank Spring Structure
     ThighModLeftLegDrag:Thigh Mod Left Leg Drag Structure
     PlayerGliderType:EGliderType Enum
     bEnableSpringMods:bool=~
     bAllowPlayerDeployRootMod:bool=~
     bUseSurfStyle:bool=~
     bUseLeanPlayRateAdjust:bool=~
     bUseForwardAs_Y_BS:bool=~
     bUseDeployCurveForIntoPlayRateAdjust:bool=~
     LeanForwardValue:float=~
     IdleValue:float=~
     LeanBackValue:float=~
     IdleInterpToSpeed:float=~
     LeanForwardInterpToSpeed:float=~
     LeanBackInterpToSpeed:float=~
    ) {is=title K2Node_BreakStruct}.PlayerGliderType
   )
   > Set has_hand_glider(has_hand_glider:bool=true) -> (Output_Get:bool)
   Event BlueprintInitializeAnimation()
   > Set fort_player(fort_player:Fort Player Anim Instance Object Reference=GetOwnerFortPlayerAnimInstance!(self:Custom Character Part Anim Instance Object Reference) -> (ReturnValue:Fort Player Anim Instance Object Reference)) -> (Output_Get:Fort Player Anim Instance Object Reference)
  ]}
]
variables [
 alt_weapon {type=boolean}
 fort_player {type=FortPlayerAnimInstance, default=null}
 has_hand_glider {type=boolean}
 is_playing_celebration {type=boolean}
]
